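# Translucent, unlit, two-sided test material in the trap assets, apparently for a range or distance box display.
# The distance mask's green channel blends two blues into an emissive-adjustment function, and separately blends a dimmed box texture toward a faint constant, likely for opacity.
# A vertex-offset-style chain scales toward a pivot by normalized distance, adds movement weighted by distance squared, and optionally applies BlendMorphs through static switches.
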Material /Game/Weapons/FORT_TRAPS/General/M_test_cube
graph {
 properties {blendMode=BLEND_Translucent, pixelDepthOffsetMode=PDOM_Legacy, shadingModel=MSM_Unlit, twoSided=true, usedShadingModels=Unlit}
 nodes [
  n1 = TextureSample() {texture=/Game/Weapons/FORT_TRAPS/General/Trap_Range/T_Trap_distance_Mask}
  LinearInterpolate(
   A=Multiply(A=TextureSample() {texture=&/T_Distance_Box}.RGB, B=Constant() {r=0.2})
   B=Constant() {r=0.05}
   Alpha=@n1.G
  )
  MaterialFunctionCall(Emissive=LinearInterpolate(A=Constant3Vector() {constant=<rgba 0 0.0835 0.9469 0>}, B=Constant3Vector() {constant=<rgba 0.042 0.1459 1 0>}, Alpha=@n1.G)) {materialFunction=&MF_FX_EmissiveAdjustments}
  n2 = MaterialFunctionCall!(In=OneMinus(None=ScalarParameter() {group=Morph Targets, parameterName=MorphControl})) {materialFunction=/Game/Effects/Fort_Effects/BluePrints/WeakPointEffect/BlendMorphs}
  n3 = VectorParameter() {parameterName=WorldSpacePivotLocation, defaultValue=<rgba 0 0 0 1>}
  n4 = WorldPosition()
  n5 = StaticSwitchParameter(True=Add(A=@n2.Result, B=@n4.XYZ), False=@n4.XYZ) {group=Morph Targets, parameterName=UseMorphTargets}
  n6 = ScalarParameter() {parameterName=InverseInnerScale}
  n7 = ScalarParameter() {parameterName=InverseOuterScale}
  n8 = Clamp(None=Divide(A=Distance(A=@n3.RGB, B=@n5), B=ScalarParameter() {defaultValue=512, parameterName=ObjectDiameter}))
  n9 = Add(
   A=Multiply(A=Subtract(A=@n3.RGB, B=@n5), B=LinearInterpolate(A=@n6, B=@n7, Alpha=@n8))
   B=LinearInterpolate(B=VectorParameter() {parameterName=InverseMovementVector, defaultValue=<rgba 0 0 0 1>}.RGB, Alpha=Multiply(A=@n8, B=@n8)) {constB=0}
  )
  StaticSwitchParameter(True=Add(A=@n2.Result, B=@n9), False=@n9) {group=Morph Targets, parameterName=UseMorphTargets}
 ]
}
